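; BP_Jungle_Parrots_Spline is a jungle environment effect Blueprint that feeds an editable spline path to a parrot Niagara system.
; The construction script passes Radius and Lifetime to the Niagara component, then calls GetBeginAndEndLocation, which BeginPlay also calls.
; Despite its name, that function collects the world location of every spline point and writes them to the Niagara SplineArray vector array.
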(event EventBeginPlay
  (CallFunction|GetBeginAndEndLocation))

(fn GetBeginandEndLocation
  (bind _spline (Variables|Default|GetSpline))
  (bind _locations (Variables|Default|GetLocations))
  (Utilities|Array|Clear _locations)
  (for _index (range (Spline|GetNumberOfSplinePoints _spline))
    (Utilities|Array|Add _locations (Spline|GetLocationatSplinePoint _spline _index "World")))
  (Niagara|NiagaraSetVectorArray (Variables|Default|GetNS_Jungle_Parrot_Spline) "SplineArray" _locations))

(fn ConstructionScript
  (bind _ns_jungle_parrot_spline (Variables|Default|GetNS_Jungle_Parrot_Spline))
  (Niagara|SetNiagaraVariable(Float) _ns_jungle_parrot_spline "Radius" (Variables|Default|GetRadius))
  (Niagara|SetNiagaraVariable(Float) _ns_jungle_parrot_spline "Lifetime" (Variables|Default|GetLifetime))
  (CallFunction|GetBeginAndEndLocation))
