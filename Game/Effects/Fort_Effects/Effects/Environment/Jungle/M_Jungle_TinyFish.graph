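# Masked, lit material for tiny jungle fish drawn as Niagara mesh particles, using Flopper fish textures.
# Hue-shifted diffuse multiplies shadow/emissive colors, blended by sun direction against a panning detail normal, with near-camera color fading, feeding emissive.
# A panning gradient texture drives world-position-offset wiggle, phased by vertex alpha and masked by vertex red for eye movement.
# Dynamic parameters control wiggle per particle, optionally hue and fade distance; switches allow particle color.
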Material /Game/Effects/Fort_Effects/Effects/Environment/Jungle/M_Jungle_TinyFish
graph {
 properties {bUsedWithNiagaraMeshParticles=true, blendMode=BLEND_Masked, pixelDepthOffsetMode=PDOM_Legacy, usedShadingModels=Default Lit}
 nodes [
  Constant() {r=1}
  DynamicParameter() {parameterIndex=1, paramNames [Hue_Shift, ColorFadeDistance, ColorFadeDesat, Param4]}
  LinearInterpolate(Alpha=MaterialFunctionCall(Distance=ScalarParameter() {defaultValue=500, group=Opacity, parameterName="Distance (S)"}) {materialFunction=&Near_Camera_Fade}.Result)
  n1 = TextureSampleParameter2D() {parameterName=DiffuseTexture, texture=/Game/Gadgets/Assets/LessEffectiveFlopper/Textures/T_Flopper_D}
  n2 = ParticleColor()
  n3 = MaterialFunctionCall(
   Hue Shift Percentage=StaticSwitchParameter(
    True=DynamicParameter() {parameterIndex=1, paramNames [Hue_Shift, ColorFadeDistance, ColorFadeDesat, Param4]}.Hue_Shift
    False=ScalarParameter() {group=Color, parameterName=Hue_Shift_Percent}
   ) {parameterName=UseDynamicParametersIndex1?}
   Texture=StaticSwitchParameter(True=Multiply(A=@n1.G, B=@n2.RGB), False=@n1.RGB) {parameterName=UseParticleColor?}
  ) {materialFunction=&HueShift}
  n4 = ParticleColor()
  n5 = VectorParameter() {parameterName=FishColor_Shadow, defaultValue=<rgba 1 0.3822 0 0>}
  n6 = VectorParameter() {parameterName=FishColor_Emissive, defaultValue=<rgba 1 0.3822 0 0>}
  n7 = CollectionParameter() {
   collection=&FortniteMaterialParameters
   parameterName=SunlightDirectionalVector
  }
  n8 = Transform(None=Normalize(VectorInput=ComponentMask(None=@n7) {b=true, g=true, r=true})) {transformSourceType=TRANSFORMSOURCE_World, transformType=TRANSFORM_Tangent}
  n9 = TextureCoordinate()
  n10 = AppendVector(
   A=Multiply(A=ScalarParameter() {defaultValue=2, group=Normal, parameterName=Tile_N_X}, B=ComponentMask(None=@n9) {r=true})
   B=Multiply(A=ComponentMask(None=@n9) {g=true}, B=ScalarParameter() {defaultValue=6, group=Normal, parameterName=Tile_N_Y})
  )
  n11 = ComponentMask(None=VectorParameter() {parameterName=Speed, defaultValue=<rgba 0.008 0.0125 0 0>}.RGB) {g=true, r=true}
  n12 = MaterialFunctionCall(
   Normal=TextureSample(UVs=Panner(Coordinate=@n10, Speed=@n11)) {samplerType=SAMPLERTYPE_Normal, texture=/Game/VisualThreatManager/StormVisuals/LargeClouds/T_StormSwirlNormal}.RGB
   Flatness=ScalarParameter() {defaultValue=0.5, group=Normal, parameterName=FlattenDetailNormal}
  ) {materialFunction=&FlattenNormal}
  n13 = LinearInterpolate(
   A=@n12.Result
   B=LinearInterpolate(A=@n12.Result, B=VertexNormalWS())
   Alpha=ScalarParameter() {defaultValue=0.1, group=Normal, parameterName=LerpVertNormal}
  )
  n14 = Power(
   Base=ConstantBiasScale(None=DotProduct(A=@n8, B=@n13)) {bias=0.5, scale=0.6}
   Exp=ScalarParameter() {defaultValue=2.4, group=Normal, parameterName=SoftenShadow}
  )
  n15 = DynamicParameter() {parameterIndex=1, paramNames [Hue_Shift, ColorFadeDistance, ColorFadeDesat, Param4]}
  n16 = ScalarParameter() {defaultValue=500, group=Color, parameterName=Color Distance}
  n17 = LinearInterpolate(
   A=LinearInterpolate(
    A=StaticSwitchParameter(True=@n4.RGB, False=@n5.RGB) {parameterName=UseParticleColor?}
    B=StaticSwitchParameter(True=@n4.RGB, False=@n6.RGB) {parameterName=UseParticleColor?}
    Alpha=Clamp(None=@n14)
   )
   B=VectorParameter() {parameterName=FishColor_FadeOut}.RGB
   Alpha=MaterialFunctionCall(Distance=StaticSwitchParameter(True=@n15.ColorFadeDistance, False=@n16) {parameterName=UseDynamicParametersIndex1?}) {materialFunction=&Near_Camera_Fade}.Result
  )
  MaterialFunctionCall(Emissive=Multiply(A=Multiply(A=@n3.Result, B=@n17), B=ScalarParameter() {defaultValue=1, group=Color, parameterName=Glow})) {materialFunction=&MF_FX_EmissiveAdjustments}
  n18 = TextureSampleParameter2D() {parameterName=SpecTexture, texture=/Game/Gadgets/Assets/LessEffectiveFlopper/Textures/T_Flopper_S}
  Multiply(A=@n18.G, B=ScalarParameter() {group=Spec, parameterName=Metallic})
  Multiply(A=@n18.B, B=ScalarParameter() {group=Spec, parameterName=Roughness})
  Multiply(A=@n18.R, B=ScalarParameter() {group=Spec, parameterName=Specular})
  TextureSampleParameter2D() {parameterName=NormalTexture, samplerType=SAMPLERTYPE_Normal, texture=/Game/Gadgets/Assets/LessEffectiveFlopper/Textures/T_Flopper_N}
  n19 = Constant3Vector()
  n20 = VertexColor()
  n21 = TextureCoordinate() {coordinateIndex=1}
  n22 = VectorParameter() {group=WPO, parameterName=TextureWPO_UVtiling, defaultValue=<rgba 1 0 0 0>}
  n23 = DynamicParameter() {paramNames [TextureWPO_Speed_Mult, TextureWPO_UVTiling_Mult, TextureWPO_Offset, TextureWPO_Intensity]}
  n24 = Multiply(A=Multiply(A=Add(A=@n20.A, B=@n21), B=ComponentMask(None=@n22.RGB) {g=true, r=true}), B=@n23.TextureWPO_UVTiling_Mult)
  n25 = VectorParameter() {group=WPO, parameterName=TextureWPO_Speed, defaultValue=<rgba 1 0 0 0>}
  n26 = AppendVector(
   A=Add(A=ComponentMask(None=@n25.RGB) {r=true}, B=LinearInterpolate(Alpha=@n20.A) {constA=0.15, constB=0.35})
   B=ComponentMask(None=@n25.RGB) {g=true}
  )
  n27 = Multiply(A=@n26, B=@n23.TextureWPO_Speed_Mult)
  n28 = ScalarParameter() {group=WPO, parameterName=TextureWPO_Per_Particle}
  n29 = TextureSample(UVs=Panner(Coordinate=Add(A=@n24, B=@n23.TextureWPO_Offset), Speed=LinearInterpolate(A=@n26, B=@n27, Alpha=@n28)) {speedX=0.15}) {texture=/Game/Global/Textures/Default/Masks_and_Gradients/WeaponGradients/T_Gradient_DesertSky_Color}
  n30 = ScalarParameter() {defaultValue=1, group=WPO, parameterName=WPO_Height}
  n31 = AppendVector(
   A=AppendVector(A=Multiply(A=@n29.R, B=@n30) {constB=0.5}, B=Multiply(A=@n29.R, B=@n30) {constB=0.5})
   B=Multiply(A=@n29.R, B=@n30) {constB=0.6}
  )
  n32 = LinearInterpolate(B=ScalarParameter() {group=WPO, parameterName=WPO_EyeMovent}, Alpha=VertexColor().R) {constA=1, constB=0}
  Transform(None=Multiply(
    A=LinearInterpolate(A=@n19, B=@n31, Alpha=@n32)
    B=DynamicParameter() {paramNames [TextureWPO_Speed_Mult, TextureWPO_UVTiling_Mult, TextureWPO_Offset, TextureWPO_Intensity]}.TextureWPO_Intensity
   ) {constB=80}) {transformSourceType=TRANSFORMSOURCE_Local}
 ]
}
