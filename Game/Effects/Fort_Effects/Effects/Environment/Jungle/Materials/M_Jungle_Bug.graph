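# Masked, two-sided lit material for jungle bug mesh particles, built on the T_Bug_01 texture.
# The texture is brightened, hue-shifted and multiplied by particle color for diffuse, while its hue-shifted blue channel is desaturated and tinted for specular, both feeding ConvertFromDiffSpec.
# An optional, invertible camera-distance fade combines with a noise texture to dissolve the alpha.
# Responsive AA is on and ray-traced shadows are off.
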Material /Game/Effects/Fort_Effects/Effects/Environment/Jungle/Materials/M_Jungle_Bug
graph {
 properties {
  bCastRayTracedShadows=false
  bEnableResponsiveAA=true
  bUsedWithMeshParticles=true
  bUsedWithNiagaraMeshParticles=true
  blendMode=BLEND_Masked
  pixelDepthOffsetMode=PDOM_Legacy
  twoSided=true
  usedShadingModels=Default Lit
 }
 nodes [
  n1 = TextureSample() {texture=/Game/Packages/Fortress_Transylvania/Transylvania_effects/Textures/T_Bug_01}
  n2 = MaterialFunctionCall(Hue Shift Percentage=Constant(), Texture=@n1.B) {materialFunction=&HueShift}
  Clamp(None=@n2.Result)
  n3 = Constant()
  n4 = Multiply(A=TextureCoordinate(), B=ScalarParameter() {defaultValue=1, group=Camera Distance Fade, parameterName=Noise Texture UV Mult})
  n5 = CameraPositionWS()
  n6 = WorldPosition()
  n7 = Divide(
   A=Subtract(A=Distance(A=@n5, B=@n6.XYZ), B=ScalarParameter() {defaultValue=150, group=Camera Distance Fade, parameterName=Distance Offset})
   B=ScalarParameter() {defaultValue=128, group=Camera Distance Fade, parameterName=Distance Divisor}
  )
  n8 = Clamp(None=LinearInterpolate(A=@n7, B=OneMinus(None=@n7), Alpha=ScalarParameter() {defaultValue=1, group=Camera Distance Fade, parameterName=Invert Camera Distance Fade}))
  n9 = Clamp(None=Add(A=TextureSample(UVs=@n4) {texture=/Game/Global/Textures/Default/Noise/Noise}.G, B=LinearInterpolate(Alpha=@n8) {constA=-1}))
  LinearInterpolate(
   A=@n1.A
   B=Multiply(A=@n1.A, B=LinearInterpolate(A=@n3, Alpha=@n9))
   Alpha=ScalarParameter() {group=Camera Distance Fade, parameterName=Use Camera Distance Fade}
  )
  n10 = Constant() {r=1.2997}
  n11 = Multiply(A=@n1.RGB) {constB=2}
  n12 = Constant() {r=2}
  MaterialFunctionCall(
   DiffuseColor=Multiply(
    A=MaterialFunctionCall(Hue Shift Percentage=@n10, Texture=@n11) {materialFunction=&HueShift}.Result
    B=ParticleColor().RGB
   )
   SpecularColor=Multiply(A=Desaturation(None=@n2.Result, Fraction=@n12), B=Constant3Vector() {constant=<rgba 0.405 0.2904 0.4012 0>})
  ) {materialFunction=&ConvertFromDiffSpec}
 ]
}
