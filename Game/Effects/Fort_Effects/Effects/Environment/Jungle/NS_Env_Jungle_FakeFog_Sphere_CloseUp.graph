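# Niagara system that renders a close-up fake fog sphere for jungle environments as a GPU mesh particle.
# A burst spawns looping particles sized by User.SphereSize and stretched on Y, with distance culling on the emitter.
# FortniteNPC time of day drives color and opacity, with rainforest fog and user master opacities scaling alpha and camera distance scaling color.
# Dynamic material parameters feed the mesh material, and a scratch input controls particle killing.
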NiagaraSystem /Game/Effects/Fort_Effects/Effects/Environment/Jungle/NS_Env_Jungle_FakeFog_Sphere_CloseUp
dependencies {
 usedDataInterfaces [&NiagaraDataInterfaceCurve, &NiagaraDataInterfaceColorCurve]
 usedDynamicInputs [/Game/Effects/Niagara/DynamicInputs/Math/FN_Fit]
 usedModules [
  &self:SystemSpawnScript.SystemScriptSource.SystemScriptGraph.NiagaraNodeAssignment_1.SetVariables_#1
  &SystemState
 ]
}
emitters {FakeFog003 {
  simTarget=GPUComputeSim
  emitterUpdateScript [
   &EmitterState(
    Enable Distance Culling=true
    Life Cycle Mode=::NewEnumerator1
    Loop Behavior=::NewEnumerator1
    Loop Duration Mode=::NewEnumerator1
    MaxDistance=25000
    Module.bUseMaxDistance=true
    Module.bUseMinDistance=false
    Reset Age On Awaken=true
    Scalability Mode=::NewEnumerator1
    UseLoopDelay=false
   )
   &SpawnBurst_Instantaneous(Use Spawn Probability=false, UseLoopCountLimit=false)
  ]
  particleSpawnScript [
   /Niagara/Modules/Spawn/Initialization/InitializeParticle(
    Color=<rgba 1 1 1 0.5>
    Lifetime=&UniformRangedFloat(Minimum=1, Maximum=1)
    Mesh Scale=&Multiply_VectorByFloat(Vector=&RandomRangeVector(Minimum=<1 1 1>, Maximum=<1 1 1>), Float=$User.SphereSize)
    Module.Write Color=true
    Module.Write Lifetime=true
    Module.Write Mass=true
    Module.Write Material Random=true
    Module.Write Position=false
    Module.Write Scale=true
    Module.Write SpriteRotation=false
    Module.Write SpriteSize=false
    Position=&SimulationPosition()
    Sprite Rotation=&UniformRangedFloat(Minimum=0, Maximum=360)
    Sprite Size=&Vector2DFromFloat(Value=&UniformRangedFloat(Minimum=10, Maximum=20))
   )
   &MeshRotationForce(Rotation=<0 0 0.25>) {enabled=false}
   &MeshRotationForce(Rotation=<0 -0.1 0>) {module=MeshRotationForce001}
   &ApplyInitialForces()
  ]
  particleUpdateScript [
   &ParticleState(
    Kill Particles When Lifetime Has Elapsed=false
    Let Infinitely Lived Particles Die When Emitter Deactivates=true
    Loop Particles Lifetime=true
   )
   &ScaleMeshSize(Scale Factor=&Multiply_VectorByFloat(Vector=<1 2 1>, Float=1))
   &SolveForcesAndVelocity(Clamp Velocity=false, Limit Acceleration=false)
   &ScaleColor(
    Scale Alpha=&Multiply_Float(
     A=&FloatFromCurve(
      FloatCurve=DataInterface() {propertyValues {Curve {keys [{interpMode=RCIM_Cubic, tangentMode=RCTM_User, time=0.5, value=1}]}}}
      Scale Curve=&Multiply_Float(
       A=&Multiply_Float(A=25, B=$NPC.FortniteNPC.RainForest_Fog_MasterOpacity)
       B=&FloatFromCurve(
        FloatCurve=DataInterface() {propertyValues {Curve {keys [<at 0 1>,<at 0.25 0.5>,<at 0.5 0.5>,<at 0.75 0.5>,<at 1 1>]}}}
        CurveIndex=Linked() {linkedVariable {name=NPC.FortniteNPC.TimeOfDay_0-1, type=NiagaraFloat}}
        Scale Curve=0.5
       )
      )
     )
     B=&FloatFromCurve(
      FloatCurve=DataInterface() {propertyValues {Curve {keys [
          {interpMode=RCIM_Cubic, tangentMode=RCTM_User, time=0, value=1}
          {interpMode=RCIM_Cubic, tangentMode=RCTM_User, time=0.5, value=1.025}
          {tangentMode=RCTM_User, time=1, value=1}
         ]}}}
      Scale Curve=&Multiply_Float(A=&Multiply_Float(A=$NPC.FortniteNPC.RainForest_Fog_MasterOpacity, B=$User.MasterOpacity), B=0.1)
     )
    )
    Scale RGB=&MakeVectorFromLinearColorRGB(Color=&MultiplyLinearColor(
      Linear Color=&ColorFromCurve(
       ColorCurve=DataInterface() {propertyValues {
         AlphaCurve {keys [{interpMode=RCIM_Cubic, tangentMode=RCTM_User, time=0, value=1}, {interpMode=RCIM_Cubic, tangentMode=RCTM_User, time=1, value=1}]}
         BlueCurve {keys [
           <at 0 0 RCIM_Cubic>
           {arriveTangent=-0.0529, interpMode=RCIM_Cubic, leaveTangent=-0.0529, tangentMode=RCTM_User, time=0, value=0.105}
           {
            arriveTangent=-0.0887
            arriveTangentWeight=0.0872
            interpMode=RCIM_Cubic
            leaveTangent=-0.0887
            tangentMode=RCTM_User
            time=0.1409
            value=0.0808
           }
           {arriveTangent=-1.1305, interpMode=RCIM_Cubic, leaveTangent=-1.1305, time=0.1623, value=0.0635}
           {
            arriveTangent=1.0728
            arriveTangentWeight=0.006
            interpMode=RCIM_Cubic
            leaveTangent=1.0728
            leaveTangentWeight=0.006
            time=0.1809
            value=0.0356
           }
           {
            arriveTangent=2.985
            arriveTangentWeight=0.0043
            interpMode=RCIM_Cubic
            leaveTangent=2.985
            leaveTangentWeight=0.0043
            time=0.2044
            value=0.1087
           }
           {
            arriveTangent=1.2065
            arriveTangentWeight=0.0021
            interpMode=RCIM_Cubic
            leaveTangent=1.2065
            leaveTangentWeight=0.0021
            time=0.2116
            value=0.1275
           }
           {
            arriveTangent=0.8355
            arriveTangentWeight=0.0404
            interpMode=RCIM_Cubic
            leaveTangent=0.8355
            leaveTangentWeight=0.0404
            time=0.3291
            value=0.2591
           }
           {
            arriveTangent=0.1385
            arriveTangentWeight=0.0562
            interpMode=RCIM_Cubic
            leaveTangent=0.1385
            leaveTangentWeight=0.0562
            time=0.33
            value=0.2264
           }
           {arriveTangent=0.2938, arriveTangentWeight=0.0238, interpMode=RCIM_Cubic, leaveTangent=0.2938, time=0.5187, value=0.2854}
           {
            arriveTangent=0.1461
            arriveTangentWeight=0.0269
            interpMode=RCIM_Cubic
            leaveTangent=0.1461
            leaveTangentWeight=0.0269
            time=0.6311
            value=0.3149
           }
           {
            arriveTangent=-0.8063
            arriveTangentWeight=0.0101
            interpMode=RCIM_Cubic
            leaveTangent=-0.8063
            leaveTangentWeight=0.0101
            time=0.7083
            value=0.3131
           }
           {
            arriveTangent=-1.4701
            arriveTangentWeight=0.0364
            interpMode=RCIM_Cubic
            leaveTangent=-1.4701
            leaveTangentWeight=0.0364
            time=0.7805
            value=0.1944
           }
           {arriveTangent=-1.2596, arriveTangentWeight=0.0289, interpMode=RCIM_Cubic, leaveTangent=-1.2596, time=0.8672, value=0.0796}
           {arriveTangent=-0.5158, interpMode=RCIM_Cubic, leaveTangent=-0.5158, time=0.8985, value=0.0458}
           {arriveTangent=0.5759, arriveTangentWeight=0.0062, interpMode=RCIM_Cubic, leaveTangent=0.5759, time=0.917, value=0.0539}
           {arriveTangent=0.4127, arriveTangentWeight=0.0096, interpMode=RCIM_Cubic, leaveTangent=0.4127, time=0.9458, value=0.073}
           {arriveTangent=0.5904, arriveTangentWeight=0.0111, interpMode=RCIM_Cubic, leaveTangent=0.5904, time=0.9792, value=0.0795}
           {arriveTangentWeight=0.0069, interpMode=RCIM_Cubic, time=1, value=0.105}
          ]}
         GreenCurve {keys [
           <at 0 0 RCIM_Cubic>
           {arriveTangent=-0.0111, interpMode=RCIM_Cubic, leaveTangent=-0.0111, tangentMode=RCTM_User, time=0, value=0.0617}
           {arriveTangent=-0.1572, arriveTangentWeight=0.0871, interpMode=RCIM_Cubic, leaveTangent=-0.1572, time=0.1409, value=0.0432}
           {arriveTangent=-0.5174, interpMode=RCIM_Cubic, leaveTangent=-0.5174, tangentMode=RCTM_User, time=0.1623, value=0.0362}
           {
            arriveTangent=0.2567
            arriveTangentWeight=0.0073
            interpMode=RCIM_Cubic
            leaveTangent=0.2567
            leaveTangentWeight=0.0073
            tangentMode=RCTM_User
            time=0.1809
            value=0.0254
           }
           {
            arriveTangent=3.9192
            arriveTangentWeight=0.0034
            interpMode=RCIM_Cubic
            leaveTangent=3.9192
            leaveTangentWeight=0.0034
            time=0.2044
            value=0.1343
           }
           {
            arriveTangent=0.5076
            arriveTangentWeight=0.0014
            interpMode=RCIM_Cubic
            leaveTangent=0.5076
            leaveTangentWeight=0.0014
            time=0.2116
            value=0.1461
           }
           {
            arriveTangent=0.1962
            arriveTangentWeight=0.0208
            interpMode=RCIM_Cubic
            leaveTangent=0.1962
            leaveTangentWeight=0.0208
            time=0.3281
            value=0.197
           }
           {
            arriveTangent=0.1292
            arriveTangentWeight=0.0206
            interpMode=RCIM_Cubic
            leaveTangent=0.1292
            leaveTangentWeight=0.0206
            time=0.33
            value=0.1693
           }
           {arriveTangent=0.2748, arriveTangentWeight=0.0238, interpMode=RCIM_Cubic, leaveTangent=0.2748, time=0.5187, value=0.2217}
           {
            arriveTangent=0.1637
            arriveTangentWeight=0.0212
            interpMode=RCIM_Cubic
            leaveTangent=0.1637
            leaveTangentWeight=0.0212
            time=0.6311
            value=0.2521
           }
           {
            arriveTangent=-0.4305
            arriveTangentWeight=0.0224
            interpMode=RCIM_Cubic
            leaveTangent=-0.4305
            leaveTangentWeight=0.0224
            time=0.7083
            value=0.2527
           }
           {
            arriveTangent=-0.8944
            arriveTangentWeight=0.0233
            interpMode=RCIM_Cubic
            leaveTangent=-0.8944
            leaveTangentWeight=0.0233
            time=0.7805
            value=0.1877
           }
           {arriveTangent=-1.0698, arriveTangentWeight=0.0289, interpMode=RCIM_Cubic, leaveTangent=-1.0698, time=0.8672, value=0.1106}
           {arriveTangent=-1.2376, interpMode=RCIM_Cubic, leaveTangent=-1.2376, time=0.8985, value=0.0615}
           {arriveTangent=0.0037, arriveTangentWeight=0.0062, interpMode=RCIM_Cubic, leaveTangent=0.0037, time=0.917, value=0.049}
           {arriveTangent=0.2047, arriveTangentWeight=0.0096, interpMode=RCIM_Cubic, leaveTangent=0.2047, time=0.9458, value=0.0617}
           {arriveTangent=-0.0002, arriveTangentWeight=0.0111, interpMode=RCIM_Cubic, leaveTangent=-0.0002, time=0.9792, value=0.0617}
           {arriveTangentWeight=0.0069, interpMode=RCIM_Cubic, time=1, value=0.0617}
          ]}
         RedCurve {keys [
           <at 0 0 RCIM_Cubic>
           {arriveTangent=0.19, interpMode=RCIM_Cubic, leaveTangent=0.19, time=0, value=0.0388}
           {arriveTangent=0.0066, arriveTangentWeight=0.0872, interpMode=RCIM_Cubic, leaveTangent=0.0066, time=0.1409, value=0.0268}
           {arriveTangent=0.8688, interpMode=RCIM_Cubic, leaveTangent=0.8688, tangentMode=RCTM_User, time=0.1623, value=0.0398}
           {
            arriveTangent=2.2289
            arriveTangentWeight=0.0287
            interpMode=RCIM_Cubic
            leaveTangent=2.2289
            leaveTangentWeight=0.0287
            tangentMode=RCTM_User
            time=0.1809
            value=0.069
           }
           {
            arriveTangent=5.4677
            arriveTangentWeight=0.004
            interpMode=RCIM_Cubic
            leaveTangent=5.4677
            leaveTangentWeight=0.004
            time=0.2044
            value=0.2412
           }
           {
            arriveTangent=-0.4704
            arriveTangentWeight=0.0033
            interpMode=RCIM_Cubic
            leaveTangent=-0.4704
            leaveTangentWeight=0.0033
            time=0.2116
            value=0.2374
           }
           {
            arriveTangent=-0.582
            arriveTangentWeight=0.0234
            interpMode=RCIM_Cubic
            leaveTangent=-0.582
            leaveTangentWeight=0.0234
            time=0.3281
            value=0.1831
           }
           {
            arriveTangent=0.0737
            arriveTangentWeight=0.0207
            interpMode=RCIM_Cubic
            leaveTangent=0.0737
            leaveTangentWeight=0.0207
            time=0.33
            value=0.1685
           }
           {arriveTangent=0.1927, arriveTangentWeight=0.0238, interpMode=RCIM_Cubic, leaveTangent=0.1927, time=0.5187, value=0.1972}
           {
            arriveTangent=0.1553
            arriveTangentWeight=0.0294
            interpMode=RCIM_Cubic
            leaveTangent=0.1553
            leaveTangentWeight=0.0294
            time=0.6311
            value=0.2266
           }
           {
            arriveTangent=-0.3554
            arriveTangentWeight=0.0173
            interpMode=RCIM_Cubic
            leaveTangent=-0.3554
            leaveTangentWeight=0.0173
            time=0.7083
            value=0.2266
           }
           {
            arriveTangent=-0.0881
            arriveTangentWeight=0.0207
            interpMode=RCIM_Cubic
            leaveTangent=-0.0881
            leaveTangentWeight=0.0207
            time=0.7805
            value=0.1735
           }
           {
            arriveTangent=-0.2866
            arriveTangentWeight=0.0155
            interpMode=RCIM_Cubic
            leaveTangent=-0.2866
            leaveTangentWeight=0.0155
            time=0.8672
            value=0.2126
           }
           {arriveTangent=-2.8032, interpMode=RCIM_Cubic, leaveTangent=-2.8032, time=0.8985, value=0.1397}
           {arriveTangent=-1.757, arriveTangentWeight=0.0062, interpMode=RCIM_Cubic, leaveTangent=-1.757, time=0.917, value=0.0729}
           {arriveTangent=-0.5697, arriveTangentWeight=0.0096, interpMode=RCIM_Cubic, leaveTangent=-0.5697, time=0.9458, value=0.0567}
           {arriveTangent=-0.3305, arriveTangentWeight=0.0111, interpMode=RCIM_Cubic, leaveTangent=-0.3305, time=0.9792, value=0.0375}
           {arriveTangentWeight=0.0069, interpMode=RCIM_Cubic, time=1, value=0.0388}
          ]}
        }}
       CurveIndex=Linked() {linkedVariable {name=NPC.FortniteNPC.TimeOfDay_0-1, type=NiagaraFloat}}
      )
      Scale Factor=&MakeVector4_FromFloat(Float=/Niagara/DynamicInputs/Scalability/ScalabilityDistanceBasedFloat(
        Curve:NiagaraDataInterfaceCurve=DataInterface() {propertyValues {Curve {keys [<at 0 1>,<at 1 0>]}}}
        CurveIndex:NiagaraFloat=$Emitter.Scalability.DistanceFraction
        Bias:NiagaraFloat=0
        Scale:NiagaraFloat=0.5
       ))
     ))
    ScaleA=true
    ScaleRGB=true
   )
   &DynamicMaterialParameters(
    Index 0 Param 4=150000
    Index 1 Param 1=&RandomRangeFloat(Minimum=0.008, Maximum=0.008)
    Index 1 Param 2=&RandomRangeFloat(Minimum=0.0012, Maximum=0.0012)
    Index 1 Param 3=&RandomRangeFloat(Minimum=1, Maximum=1)
    Index 1 Param 4=&RandomRangeFloat(Minimum=2, Maximum=2)
    Index 2 Param 1=&RandomRangeFloat(Minimum=3500, Maximum=3500)
    Index 2 Param 2=&RandomRangeFloat(Minimum=1, Maximum=1)
    Index 2 Param 3=5000
    Index 2 Param 4=&Multiply_Float(A=&RandomRangeFloat(Minimum=1, Maximum=1), B=150)
    Module.Param0WriteEnabled=true
    Module.Param0WriteEnabled1=true
    Module.Param0WriteEnabled2=true
    Module.Param0WriteEnabled3=true
    Module.Param1WriteEnabled=true
    Module.Param1WriteEnabled1=true
    Module.Param1WriteEnabled2=true
    Module.Param1WriteEnabled3=true
    Module.Param2WriteEnabled=true
    Module.Param2WriteEnabled1=true
    Module.Param2WriteEnabled2=true
    Module.Param2WriteEnabled3=true
    Module.Param3WriteEnabled=true
    Module.Param3WriteEnabled1=true
    Module.Param3WriteEnabled2=true
    Module.Param3WriteEnabled3=true
    Write Parameter Index 1=true
    Write Parameter Index 2=true
    Write Parameter Index 3=true
   )
   &CameraOffset()
   &KillParticles(Kill Particles=&self:ScratchDynamicInput())
  ]
  rendererClasses [&NiagaraMeshRendererProperties]
  renderers [{rendererClass=&NiagaraMeshRendererProperties, rendererIndex=0}]
 }}
summary {
 systemName=NS_Env_Jungle_FakeFog_Sphere_CloseUp
 emitters [{emitterName=FakeFog003, simTarget=GPUComputeSim, rendererClasses [&NiagaraMeshRendererProperties]}]
 userVariables [
  User.MasterOpacity {type=NiagaraFloat, defaultValue=250, description=""}
  User.SphereSize {type=NiagaraFloat, defaultValue=25, description=""}
 ]
}
