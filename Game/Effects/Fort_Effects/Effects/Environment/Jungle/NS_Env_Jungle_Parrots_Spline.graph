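# Niagara system for an ambient jungle parrot flock: GPU mesh particles that travel relative to a spline path.
# A random burst plus a brief tapering spawn rate places birds around a User.Radius sphere, with dynamic material parameters set per particle.
# A scratch module takes User.SplineArray points and a lerp curve, while a point force centred on Particles.SplineForce steers motion alongside curl noise and drag.
# Mesh scale grows in and shrinks out over each particle's life.
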NiagaraSystem /Game/Effects/Fort_Effects/Effects/Environment/Jungle/NS_Env_Jungle_Parrots_Spline
dependencies {usedDataInterfaces [&NiagaraDataInterfaceCurve], usedDynamicInputs [], usedModules [&SystemState]}
emitters {Parrots {
  simTarget=GPUComputeSim
  emitterUpdateScript [
   &EmitterState()
   &SpawnRate(
    SpawnRate=&Multiply_Float(
     A=100
     B=&FloatFromCurve(FloatCurve=DataInterface() {propertyValues {Curve {keys [<at 0 1>,<at 0.1 0>]}}}, CurveIndex=$Emitter.NormalizedLoopAge)
    )
    Use Spawn Probability=false
   )
   &SpawnBurst_Instantaneous(
    Spawn Count=&RandomRangeInt(Minimum=5, Maximum=10)
    Use Spawn Probability=false
    UseLoopCountLimit=false
   )
  ]
  particleSpawnScript [
   &InitializeParticle(Color Mode=::NewEnumerator0, Lifetime=10, Mesh Scale Mode=::NewEnumerator3, UsePositionOffset=false, Write Material Random=true)
   &ShapeLocation(MaskSpawnGroup=false, Override Seed=false, Sphere Radius=$User.Radius, UseOwnerScale=true)
   &DynamicMaterialParameters(
    Index 0 Param 1=0.25
    Index 0 Param 2=&Multiply_Float(A=$Particles.Lifetime, B=&RandomRangeFloat(Minimum=0.16, Maximum=0.18))
    Index 0 Param 3=0.25
    Index 0 Param 4=0.1
    Index 1 Param 1=1
    Index 1 Param 2=1
    Index 1 Param 3=1
    Index 2 Param 1=1
    Index 2 Param 2=1
    Index 3 Param 1=1
    Index 3 Param 4=1
    Module.Param0WriteEnabled=true
    Module.Param0WriteEnabled1=true
    Module.Param0WriteEnabled2=true
    Module.Param0WriteEnabled3=true
    Module.Param1WriteEnabled=true
    Module.Param1WriteEnabled1=true
    Module.Param1WriteEnabled2=true
    Module.Param1WriteEnabled3=true
    Module.Param2WriteEnabled=true
    Module.Param2WriteEnabled1=true
    Module.Param2WriteEnabled2=true
    Module.Param2WriteEnabled3=true
    Module.Param3WriteEnabled=true
    Module.Param3WriteEnabled1=true
    Module.Param3WriteEnabled2=true
    Module.Param3WriteEnabled3=true
    Write Parameter Index 1=true
    Write Parameter Index 2=true
    Write Parameter Index 3=true
   )
  ]
  particleUpdateScript [
   &self:ScratchModule(
    PositionLerp:NiagaraFloat=&FloatFromCurve(FloatCurve=DataInterface() {propertyValues {Curve {keys [{tangentMode=RCTM_User, time=0, value=0}, {tangentMode=RCTM_User, time=1, value=1}]}}})
    SplinePointPositionArray:NiagaraDataInterfaceArrayFloat3=$User.SplineArray
   ) {module=ScratchModule}
   &ParticleState()
   &ScaleMeshSize(Scale Factor=&Multiply_VectorByFloat(
     Vector=<1 1 1>
     Float=&FloatFromCurve(FloatCurve=DataInterface() {propertyValues {Curve {keys [
          <at 0 0 RCIM_Cubic>
          {arriveTangent=1.25, leaveTangent=1.25, time=0.25, value=1}
          {arriveTangent=-1.3333, interpMode=RCIM_Cubic, leaveTangent=-1.3333, time=0.85, value=1}
          <at 1 0 RCIM_Cubic>
         ]}}})
    ))
   &CurlNoiseForce(
    Noise Frequency=4
    Noise Strength=&Multiply_Float(A=$Particles.Lifetime, B=40)
    Pan Noise=true
    Pan Noise Field=&MakeVector(
     X=&Multiply_Float(
      A=&RandomRangeFloat(Minimum=-10, Maximum=10)
      B=&Divide_Float(A=$Particles.Lifetime, B=10)
     )
     Y=&Multiply_Float(
      A=&RandomRangeFloat(Minimum=-10, Maximum=10)
      B=&Divide_Float(A=$Particles.Lifetime, B=10)
     )
     Z=&Multiply_Float(
      A=&RandomRangeFloat(Minimum=-10, Maximum=10)
      B=&Divide_Float(A=$Particles.Lifetime, B=10)
     )
    )
   )
   &PointForce(
    Force Falloff Distance=15000
    Force Origin=$Particles.SplineForce
    Force Strength=&OneMinusFloat(Float=&Multiply_Float(A=$Particles.Lifetime, B=250))
    Module.Use Falloff=true
    Module.UseLinearForce=false
   )
   &Drag(Drag=&Divide_Float(A=$Particles.Lifetime, B=100), Use Linear Drag=true, Use Rotational Drag=true)
   &SolveForcesAndVelocity(Clamp Velocity=true, Limit Acceleration=true)
  ]
  rendererClasses [&NiagaraMeshRendererProperties]
  renderers [{rendererClass=&NiagaraMeshRendererProperties, rendererIndex=0}]
 }}
summary {
 systemName=NS_Env_Jungle_Parrots_Spline
 emitters [{emitterName=Parrots, simTarget=GPUComputeSim, rendererClasses [&NiagaraMeshRendererProperties]}]
 userVariables [
  User.Lifetime {type=NiagaraFloat, defaultValue=0.1, description=""}
  User.Radius {type=NiagaraFloat, defaultValue=100, description=""}
  User.Spline {
   type=NiagaraDataInterfaceSpline
   description=""
   defaultValue {
    struct=NiagaraExt_VariableValue_DataInterface
    value {
     dataInterface=&self:NiagaraDataInterfaceSpline_0
     dataInterfaceClass=/Script/Niagara.NiagaraDataInterfaceSpline
    }
   }
  }
  User.SplineArray {
   type=NiagaraDataInterfaceArrayFloat3
   description=""
   defaultValue {
    struct=NiagaraExt_VariableValue_DataInterface
    value {
     dataInterface=&self:NiagaraDataInterfaceArrayFloat3_0
     dataInterfaceClass=&NiagaraDataInterfaceArrayFloat3
    }
   }
  }
 ]
}
