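# Niagara GPU system for jungle vegetation debris on water, spawning green mesh leaves across a user-sized cylinder and drifting them with distance culling.
# Curl noise, wind, player-driven point force and distance-field avoidance feed a horizontal-only solve, then curve-driven gravity feeds a second solve.
# User parameters set spawn rate, radius and fall/curl curves.
# A disabled Petals_Low emitter holds a flat pink sprite variant with time-of-day tinting and a spawn cap.
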NiagaraSystem /Game/Effects/Fort_Effects/Effects/Environment/Jungle/NS_Env_Jungle_Vegetation_OnWater
dependencies {
 usedDataInterfaces [
  &NiagaraDataInterfaceMeshRendererInfo
  &NiagaraDataInterfaceCollisionQuery
  /Script/ChaosNiagara.NiagaraDataInterfacePhysicsField
  &NiagaraDataInterfaceLandscape
  &NiagaraDataInterfaceSpriteRendererInfo
 ]
 usedDynamicInputs [
  &MakeFloatFromVector
  &self:ScratchDynamicInput
 ]
 usedModules [
  &SystemState
  &self:SystemSpawnScript.SystemScriptSource.SystemScriptGraph.NiagaraNodeAssignment_0.SetVariables_#1
  &self:SystemSpawnScript.SystemScriptSource.SystemScriptGraph.NiagaraNodeAssignment_1.SetVariables_#2
 ]
}
emitters {
 Leaves_HIGH {
  simTarget=GPUComputeSim
  emitterUpdateScript [
   &EmitterState(
    Enable Distance Culling=true
    MaxDistance=10000
    Module.bApplySpawnCountScaleByDistance=false
    Module.bUseMaxDistance=true
    Module.bUseMinDistance=false
    Scalability Mode=::NewEnumerator1
    Scale Spawn Count=true
    Spawn Count Scale By Distance=DataInterface() {propertyValues {Curve {keys [<at 0 0>,<at 0.2 1>,<at 0.75 1>,<at 1 0>]}}}
   )
   &SpawnRate(SpawnRate=$User.SpawnRate, Use Spawn Probability=false)
  ]
  particleSpawnScript [
   &InitializeParticle(
    Color=&MultiplyLinearColor(
     Linear Color=<rgba 0.0267 0.1146 0.0013 1>
     Scale Factor=&MakeVector4_FromFloat(Float=&RandomRangeFloat(Minimum=1, Maximum=1))
    )
    Lifetime Max=6
    Lifetime Min=5
    Lifetime Mode=::NewEnumerator1
    Mesh Renderer Array Visibility Mode=::NewEnumerator3
    Mesh Renderer Info=DataInterface() {propertyValues {MeshRenderer=&self:Leaves_HIGH_0.NiagaraMeshRendererProperties_1}}
    Mesh Scale Max=<0.07 0.07 0.01>
    Mesh Scale Min=<0.04 0.04 0.005>
    Mesh Scale Mode=::NewEnumerator2
    Sprite Size=<28 20>
    Sprite Size Mode=::NewEnumerator4
    UsePositionOffset=false
    Write Material Random=true
   )
   &ShapeLocation(
    Cylinder Height=1
    Cylinder Height Midpoint=0.2
    Cylinder Radius=$User.RadiusSpread
    MaskSpawnGroup=false
    Offset=<0 0 -0.5>
    Offset Mode=::NewEnumerator0
    Override Seed=false
    Shape Primitive=::NewEnumerator1
    Surface Only=true
    Surface Only Band Thickness=1
    Use Endcaps In Surface Only Mode=false
    UseOwnerScale=true
   )
   &AddVelocity(Velocity=&RandomRangeVector(Minimum=<0 -35 0>, Maximum=<0 -40 0>))
   &ApplyInitialForces()
   &InitialMeshOrientation(
    Mesh Axis To Orient=<0 1 0>
    Mesh Orientation Mode=::NewEnumerator2
    Rotation=&RandomRangeVector(Minimum=<0 0 0>, Maximum=<0 0 1>)
    Use Rotation Vector=true
   )
   /SetVariables(Particles.VelocityClamp:Vector3f=&Multiply_Vector(A=$Particles.Velocity, B=<1 1 0>)) {isSetParameters=true, module=SetVariables_#3}
   &KillParticles(Kill Particles=&SetBoolByIntComparison(A=$Engine.Emitter.TotalSpawnedParticles, B=$User.SpawnCap, Comparison Type=::NewEnumerator0)) {enabled=false}
  ]
  particleUpdateScript [
   &ParticleState()
   &ScaleColor(
    Scale RGB=&VectorFromFloat(Value=&FloatFromCurve(FloatCurve=DataInterface() {propertyValues {Curve {keys [<at 0.8 1>,<at 1 0>]}}}))
    ScaleA=true
    ScaleRGB=true
   )
   &PointForce(
    Force Strength=&RandomRangeFloat(Minimum=0.25, Maximum=0.5)
    Module.Use Falloff=false
    Module.UseLinearForce=false
   ) {module=PointForce001}
   &MeshRotationForce(Rotation=&MakeVector(
     X=&Multiply_Float(A=1, B=&FloatFromCurve(FloatCurve=$User.FallCurve))
     Y=&Multiply_Float(A=1, B=&FloatFromCurve(FloatCurve=$User.FallCurve))
     Z=&Multiply_Float(A=1, B=&FloatFromCurve(FloatCurve=$User.FallCurve))
    )) {module=MeshRotationForce001}
   &KillParticlesInVolume(Invert Volume=true, Sphere Radius:NiagaraFloat=&Multiply_Float(A=$User.RadiusSpread, B=5))
   /Niagara/Modules/Math/AvoidDistanceFieldSurfaces_GPU(
    Avoid Nearest Surface=true
    Avoid Oncoming Surfaces=true
    Falloff Distance=5000
    Max Trace Distance=2500
    Nearest Surface Avoidance Strength=50
    Obstactle Avoidance Strength=1
   )
   &InterpolateOverTime(Rate Of Change=2, Target Vector Value:Vector3f=$Particles.Velocity, Variable Type=::NewEnumerator2)
   &MeshRotationForce(Rotation=&MakeVector(
     X=0
     Y=0
     Z=&Multiply_Float(
      A=&RandomRangeFloat(Minimum=-0.005, Maximum=0.005)
      B=&VectorLength(Vector=$Particles.InterpolateOverTime.InterpolateOverTime_V3_.Moving Average)
     )
    ))
   &CurlNoiseForce(
    Noise Frequency=2
    Noise Strength=&Multiply_Float(
     A=&Multiply_Float(
      A=&RandomRangeFloat(Minimum=10, Maximum=50)
      B=&FloatFromCurve(FloatCurve=$User.CurlCurve)
     )
     B=1
    )
    Pan Noise=true
    Pan Noise Field=&RandomRangeVector(Minimum=<-0.05 0 0.1>, Maximum=<0.05 0.25 1>)
   )
   &WindForce(Scale=150, Use Ground Mask=true, Use Physics Field=true, Wind Speed=<0 250 0>, Wind Speed Scale=250)
   &Drag(Drag=0, Rotational Drag=0, Use Linear Drag=true, Use Rotational Drag=true)
   &SolveForcesAndVelocity(
    Clamp Velocity=true
    Force:Vector3f=&Multiply_Vector(A=$Transient.PhysicsForce, B=<1 1 0>)
    Limit Acceleration=false
    Speed Limit=15000
    Velocity:Vector3f=&Multiply_Vector(A=$Particles.Velocity, B=<1 1 1>)
   ) {enabled=false}
   &PointForce(
    Force Falloff Distance:NiagaraFloat=&RandomRangeFloat(Minimum=500, Maximum=500)
    Force Origin=&AddVectorToPosition(Position=$System.PlayerPosition, Vector=<0 0 0>)
    Force Strength=&RandomRangeFloat(Minimum=500, Maximum=500)
    Linear Force Vector:Vector3f=$System.PlayerVelocity
    Module.Use Falloff=true
    Module.UseLinearForce=true
   )
   /Niagara/Modules/Update/Forces/PointAttractionForce(
    Attraction Radius:NiagaraFloat=&Waveform(
     Waveform Count=::NewEnumerator1
     Global Amplitude Scale=250
     "Waveform [1]"=::NewEnumerator0
     "[1] Frequency"=5
     "[1] Amplitude Min/Max"=<0.4 1>
     "[2] Average"=false
     "[2] Input Position"=0
     "[2] Phase"=$Emitter.Age
     "[2] Phase Scale"=1
     "[2] Compound Frequency"=2
     "[2] Frequency"=2
     "[2] Amplitude Min/Max"=<0 1>
     "[2] Scale"=1
     "[2] Bias"=0
     "[2] Exponent"=1
     "[2] Pulse Duration"=0.1
     "[2] Clip Triangle Min/Max"=<0 0>
     "[2] Random Seed"=0
    )
    AttractionStrength:NiagaraFloat=&Multiply_Float(
     A=&FloatFromCurve(FloatCurve=DataInterface() {propertyValues {Curve {keys [<at 0 1>,<at 0.25 0>]}}}, Scale Curve=-250)
     B=1
    )
    AttractorPosition:Vector3f=&Add_Vector(
     A=$Emitter.CameraPosition
     B=&Add_Vector(
      A=&Add_Vector(
       A=&Multiply_VectorByFloat(Vector=<1 1 1>, Float=-130)
       B=&Multiply_VectorByFloat(Vector=$Emitter.CameraForward, Float=250)
      )
      B=&Multiply_VectorByFloat(Vector=$Emitter.CameraRight, Float=-50)
     )
    )
    Falloff Exponent=1
    Module.Kill Within Radius=false
    Module.Use Falloff=false
   ) {enabled=false, module=PointAttractionForce001}
   &SolveForcesAndVelocity(
    Acceleration Limit=25
    Clamp Velocity=true
    Force:Vector3f=&Multiply_Vector(A=$Transient.PhysicsForce, B=<1 1 0>)
    Limit Acceleration=true
    Rotational Velocity:Vector3f=&Multiply_Vector(A=$Particles.RotationalVelocity, B=<1 1 1>)
    Speed Limit=50
   ) {module=SolveForcesAndVelocity001}
   &AlignSpriteToMeshOrientation(
    Mesh Orientation Relative Sprite Alignment Vector=<1 0 0>
    Mesh Orientation Relative Sprite Facing Vector:Vector3f=<0 0 1>
    SetAlignment=true
    SetFacing=true
   )
   &GravityForce(Gravity=&MakeVector(X=0, Y=0, Z=&FloatFromCurve(FloatCurve=$User.FallCurve, Scale Curve=-150)))
   &SolveForcesAndVelocity(Acceleration Limit=25, Clamp Velocity=true, Limit Acceleration=true, Speed Limit=100) {module=SolveForcesAndVelocity002}
   &ScaleMeshSize(Scale Factor=&VectorFromFloat(Value=&FloatFromCurve(
      FloatCurve=DataInterface() {propertyValues {Curve {keys [
          {interpMode=RCIM_Cubic, tangentMode=RCTM_User, time=0, value=0}
          <at 0.1 10>
          {interpMode=RCIM_Cubic, tangentMode=RCTM_User, time=0.9, value=10}
          {tangentMode=RCTM_User, time=1, value=0}
         ]}}}
      Scale Curve=1.25
     )))
  ]
  rendererClasses [&NiagaraMeshRendererProperties]
  renderers [{rendererClass=&NiagaraMeshRendererProperties, rendererIndex=0}]
 }
 Petals_Low {
  bEnabled=false
  simTarget=GPUComputeSim
  emitterUpdateScript [
   &EmitterState(
    Enable Distance Culling=true
    Life Cycle Mode=::NewEnumerator1
    Loop Duration=1
    Module.bApplySpawnCountScaleByDistance=false
    Module.bUseMaxDistance=true
    Module.bUseMinDistance=true
    Scalability Mode=::NewEnumerator1
    Scale Spawn Count=true
    Spawn Count Scale By Distance=DataInterface() {propertyValues {Curve {keys [<at 0 0>,<at 0.2 1>,<at 0.75 1>,<at 1 0>]}}}
    UseLoopDelay=false
   )
   &SpawnRate(SpawnRate=$User.SpawnRate, Use Spawn Probability=false)
  ]
  particleSpawnScript [
   &InitializeParticle(
    Color=&RandomRangeLinearColor(Minimum=<rgba 0.7011 0.3515 0.4397 1>, Maximum=<rgba 0.9219 0.53 0.6417 1>)
    Lifetime Max=5
    Lifetime Min=2
    Lifetime Mode=::NewEnumerator1
    Mesh Scale Mode=::NewEnumerator3
    Mesh Uniform Scale=0.025
    Sprite Size=<18 10>
    Sprite Size Mode=::NewEnumerator4
    UsePositionOffset=false
    Write Material Random=true
   )
   &ShapeLocation(
    Cylinder Height=0
    Cylinder Radius=$User.RadiusSpread
    MaskSpawnGroup=false
    Non Uniform Scale=<1 1 0>
    Override Seed=false
    Shape Primitive=::NewEnumerator1
    Surface Only=true
    Surface Only Band Thickness:NiagaraFloat=&RandomRangeFloat(Minimum=0.4, Maximum=0.5)
    Use Endcaps In Surface Only Mode=false
    UseOwnerScale=true
   )
   &ApplyInitialForces()
   &InitialMeshOrientation(
    Mesh Axis To Orient=<0 1 0>
    Mesh Orientation Mode=::NewEnumerator2
    Rotation=&RandomRangeVector(Minimum=<0 0 0>, Maximum=<0 0 1>)
    Use Rotation Vector=true
   )
   /SetVariables(Particles.VelocityClamp:Vector3f=&Multiply_Vector(A=$Particles.Velocity, B=<1 1 0>)) {isSetParameters=true, module=SetVariables_#3}
   &KillParticles(Kill Particles=&SetBoolByIntComparison(A=$Engine.Emitter.TotalSpawnedParticles, B=$User.SpawnCap, Comparison Type=::NewEnumerator0))
  ]
  particleUpdateScript [
   &ParticleState(Kill Particles When Lifetime Has Elapsed=false)
   &ScaleColor(
    Scale RGB=&FN_Select_Vector_from_Time_Of_Day(
     Morning=&MakeVectorFromLinearColorRGB(Color=<rgba 0.6308 0.491 0.4969 1>)
     Day=&MakeVectorFromLinearColorRGB(Color=<rgba 1 1 1 1>)
     Evening=&MakeVectorFromLinearColorRGB(Color=<rgba 0.6867 0.4735 0.4508 1>)
     Night=&MakeVectorFromLinearColorRGB(Color=<rgba 0.3376 0.4207 0.4635 1>)
    )
    ScaleA=true
    ScaleRGB=true
   )
   &PointForce(
    Force Strength=&RandomRangeFloat(Minimum=0.25, Maximum=0.5)
    Module.Use Falloff=false
    Module.UseLinearForce=false
   ) {module=PointForce001}
   &KillParticlesInVolume(Invert Volume=true, Sphere Radius:NiagaraFloat=$User.RadiusSpread)
   &CurlNoiseForce(
    Noise Frequency=2
    Noise Strength=&RandomRangeFloat(Minimum=10, Maximum=20)
    Pan Noise=true
    Pan Noise Field=&RandomRangeVector(Minimum=<-0.05 -0.05 0.1>, Maximum=<0.05 0.05 0.05>)
   )
   &Drag(
    Drag=&RandomRangeFloat(Minimum=1, Maximum=3)
    Rotational Drag=20
    Use Linear Drag=true
    Use Rotational Drag=true
   )
   &SolveForcesAndVelocity(
    Clamp Velocity=false
    Force:Vector3f=&Multiply_Vector(A=$Transient.PhysicsForce, B=<1 1 0>)
    Limit Acceleration=false
    Velocity:Vector3f=&Multiply_Vector(A=$Particles.Velocity, B=<1 1 0>)
   )
   &AlignSpriteToMeshOrientation(
    Mesh Orientation Relative Sprite Alignment Vector=<1 0 0>
    Mesh Orientation Relative Sprite Facing Vector:Vector3f=<0 0 1>
    SetAlignment=true
    SetFacing=true
   )
   &DynamicMaterialParameters(
    Index 0 Param 1=0.2
    Module.Param0WriteEnabled=true
    Module.Param1WriteEnabled=true
    Module.Param2WriteEnabled=true
    Module.Param3WriteEnabled=true
   )
   &SubUVAnimation(
    Direct Frame Index:NiagaraFloat=&RandomRangeFloat(Minimum=1, Maximum=6)
    "Sprite Renderer "=DataInterface() {propertyValues {SpriteRenderer=&self:Petals_Low_7.NiagaraSpriteRendererProperties_0}}
    SubUV Animation Mode:ENiagara_SubUVLookupModeV2=::NewEnumerator5
   )
  ]
  rendererClasses [&NiagaraSpriteRendererProperties]
  renderers [{rendererClass=&NiagaraSpriteRendererProperties}]
 }
}
summary {
 systemName=NS_Env_Jungle_Vegetation_OnWater
 emitters [
  {emitterName=Leaves_HIGH, simTarget=GPUComputeSim, rendererClasses [&NiagaraMeshRendererProperties]}
  {bEnabled=false, emitterName=Petals_Low, simTarget=GPUComputeSim, rendererClasses [&NiagaraSpriteRendererProperties]}
 ]
 userVariables [
  User.CurlCurve {
   type=NiagaraDataInterfaceCurve
   description=""
   defaultValue {
    struct=NiagaraExt_VariableValue_DataInterface
    value {
     dataInterface=&self:NiagaraDataInterfaceCurve_1
     dataInterfaceClass=&NiagaraDataInterfaceCurve
    }
   }
  }
  User.FallCurve {
   type=NiagaraDataInterfaceCurve
   description=""
   defaultValue {
    struct=NiagaraExt_VariableValue_DataInterface
    value {
     dataInterface=&self:NiagaraDataInterfaceCurve_0
     dataInterfaceClass=&NiagaraDataInterfaceCurve
    }
   }
  }
  User.RadiusSpread {type=NiagaraFloat, defaultValue=1000, description=""}
  User.SpawnCap {type=NiagaraInt32, defaultValue=600}
  User.SpawnRate {type=NiagaraFloat, defaultValue=15, description=""}
 ]
}
